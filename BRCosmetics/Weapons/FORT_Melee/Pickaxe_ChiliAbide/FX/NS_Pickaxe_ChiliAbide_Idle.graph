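# Niagara idle effect for the ChiliAbide pickaxe that plays a mesh sheen followed by a sprite sparkle.
# Each emitter burst-spawns one particle when User.AlwaysOnToggle is 0 and none when it is 1.
# The sheen animates a dynamic material parameter by curve, while the sparkle spawns offset before the sheen lifetime ends, then pops, spins and shrinks.
# Both use time-of-day color and fade alpha by the system Timeline's inverse normalized time.
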NiagaraSystem /BRCosmetics/Weapons/FORT_Melee/Pickaxe_ChiliAbide/FX/NS_Pickaxe_ChiliAbide_Idle
dependencies {
 usedDataInterfaces [&NiagaraDataInterfaceArrayInt32, &NiagaraDataInterfaceCurve]
 usedDynamicInputs [&Lerp_Float]
 usedModules [
  &SystemState
  &Timeline
  &self:SystemSpawnScript.SystemScriptSource.SystemScriptGraph.NiagaraNodeAssignment_0.SetVariables_#1
 ]
}
emitters {
 Sheen_Mesh {
  emitterUpdateScript [
   &EmitterState()
   &SpawnBurst_Instantaneous(
    Spawn Count=&SelectIntFromArray(
     Direct Array Index=&MakeIntFromFloat(Float=$User.AlwaysOnToggle)
     Int Selection Array=DataInterface() {propertyValues {IntData [1,0]}}
    )
    Use Spawn Probability=false
    UseLoopCountLimit=false
   )
  ]
  particleSpawnScript [&InitializeParticle(
    Color=/Game/Effects/Niagara/DynamicInputs/TimeOfDay/DaySequenceTimeOfDayColor(Linear Color:LinearColor=<rgba 0.4705 0.4705 0.4705 1>, Time Of Day Influence:NiagaraFloat=1)
    Lifetime=$System.SheenLifetime
    Mesh Scale Mode=::NewEnumerator3
    Position Offset Coordinate Space=::Local
    UsePositionOffset=true
    Write Material Random=true
   )]
  particleUpdateScript [
   &ParticleState()
   &DynamicMaterialParameters(
    Index 0 Param 1=&FloatFromCurve(FloatCurve=DataInterface() {propertyValues {Curve {keys [{interpMode=RCIM_Cubic, tangentMode=RCTM_User, time=0, value=0}, {interpMode=RCIM_Cubic, tangentMode=RCTM_User, time=1, value=1}]}}})
    Index 0 Param 2=0.2
    Module.Index0Float0Write=true
    Module.Index0Float1Write=true
    Module.Index0Float2Write=true
    Module.Index0Float3Write=true
   ) {module=DynamicMaterialParameters001}
   &ScaleColor(Scale Alpha=$System.Timeline.InverseNormalizedTime, ScaleA=true, ScaleRGB=true)
  ]
  rendererClasses [&NiagaraMeshRendererProperties]
  renderers [{rendererClass=&NiagaraMeshRendererProperties, rendererIndex=0}]
 }
 Sparkle {
  emitterUpdateScript [
   &EmitterState()
   &SpawnBurst_Instantaneous(
    Spawn Count=&SelectIntFromArray(
     Direct Array Index=&MakeIntFromFloat(Float=$User.AlwaysOnToggle)
     Int Selection Array=DataInterface() {propertyValues {IntData [1,0]}}
    )
    Spawn Time=&Subtract_Float(A=$System.SheenLifetime, B=$System.SparkleSpawnTimeOffset)
    Use Spawn Probability=false
    UseLoopCountLimit=false
   )
  ]
  particleSpawnScript [&InitializeParticle(
    Color=&DaySequenceTimeOfDayColor(Linear Color:LinearColor=<rgba 0.4705 0.4705 0.4705 1>, Time Of Day Influence:NiagaraFloat=1)
    Lifetime=&RandomRangeFloat(Minimum=0.666, Maximum=0.777)
    Position Offset=$System.SparklePositionOffset
    Sprite Rotation Mode=::NewEnumerator2
    Sprite Size=&Multiply_Vector2D_ByFloat(Float=&RandomRangeFloat(Minimum=12, Maximum=13), Vector2D=<1 1>)
    Sprite Size Mode=::NewEnumerator4
    UsePositionOffset=true
    Write Material Random=true
   )]
  particleUpdateScript [
   &ParticleState()
   &ScaleSpriteSize(
    Uniform Curve Scale=4
    Uniform Curve Sprite Scale=DataInterface() {propertyValues {Curve {keys [
        {arriveTangent=9.0725, interpMode=RCIM_Cubic, leaveTangent=9.0725, tangentMode=RCTM_User, time=0, value=0}
        {arriveTangent=-1.7917, interpMode=RCIM_Cubic, leaveTangent=-1.7917, tangentMode=RCTM_User, time=0.0666, value=0.9847}
        {arriveTangent=-0.0405, interpMode=RCIM_Cubic, leaveTangent=-0.0405, tangentMode=RCTM_User, time=1, value=0}
       ]}}}
   )
   &SpriteRotationRate(Rotation Rate=&RandomRangeFloat(Minimum=90, Maximum=360))
   &CameraOffset(Camera Offset Amount=0.5)
   &ScaleColor(Scale Alpha=$System.Timeline.InverseNormalizedTime, ScaleA=true, ScaleRGB=true)
  ]
  rendererClasses [&NiagaraSpriteRendererProperties]
  renderers [{rendererClass=&NiagaraSpriteRendererProperties}]
 }
}
summary {
 systemName=NS_Pickaxe_ChiliAbide_Idle
 emitters [
  {emitterName=Sheen_Mesh, rendererClasses [&NiagaraMeshRendererProperties]}
  {emitterName=Sparkle, rendererClasses [&NiagaraSpriteRendererProperties]}
 ]
 userVariables [
  User.AlwaysOnToggle {type=NiagaraFloat, defaultValue=0, description=""}
  User.bIsFrontEnd {type=NiagaraBool}
  User.bIsFrontEndPreview {type=NiagaraBool}
  User.bIsSwinging {type=NiagaraBool}
 ]
}
