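# Niagara system for the swing trail of the ChiliAbide pickaxe cosmetic.
# Two ribbon emitters, Soft_Trail (faint) and ToonTrail (near-opaque), use anim-trail modules to spawn short-lived particles between different bone or socket index pairs.
# Each fades alpha by a lifetime curve times a system AlphaFade value and tints RGB by day-sequence time of day.
# User variables expose front-end flags, ribbon ID offset and swing section index.
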NiagaraSystem /BRCosmetics/Weapons/FORT_Melee/Pickaxe_ChiliAbide/FX/NS_Pickaxe_ChiliAbide_Trail
dependencies {
 usedDataInterfaces [&NiagaraDataInterfaceCurve, &NiagaraDataInterfaceSocketReader]
 usedDynamicInputs [&MergeHelper_FloatSwitch]
 usedModules [
  &SystemState
  &self:SystemSpawnScript.SystemScriptSource.SystemScriptGraph.NiagaraNodeAssignment_1.SetVariables_#1
  &self:SystemSpawnScript.SystemScriptSource.SystemScriptGraph.NiagaraNodeAssignment_3.SetVariables_#2
 ]
}
emitters {
 Soft_Trail {
  emitterSpawnScript [&FN_RibbonIDOffsetUtility()]
  emitterUpdateScript [
   &EmitterState()
   &FN_SpawnAnimTrail()
  ]
  particleSpawnScript [
   &FN_InitializeAnimTrailBoneIndecies(First Bone or Socket Index=1, Second Bone or Socket Index=0)
   &FN_InitializeAnimTrail(Color=<rgba 0.4045 0.4045 0.4045 0.25>, Lifetime=0.18)
  ]
  particleUpdateScript [
   &SolveForcesAndVelocity(Clamp Velocity=false, Limit Acceleration=false)
   &FN_UpdateAnimTrail() {module=FN_UpdateAnimTrail001}
   &ScaleColor(
    Scale Alpha=&Multiply_Float(
     A=&FloatFromCurve(FloatCurve=DataInterface() {propertyValues {Curve {keys [<at 0.1 1 RCIM_Cubic>, {arriveTangent=-2.5031, tangentMode=RCTM_User, time=1, value=0}]}}})
     B=$System.AlphaFade
    )
    Scale RGB=&MakeVectorFromLinearColorRGB(Color=/Game/Effects/Niagara/DynamicInputs/TimeOfDay/DaySequenceTimeOfDayColor(Linear Color:LinearColor=<rgba 1 1 1 1>, Time Of Day Influence:NiagaraFloat=1))
    ScaleA=true
    ScaleRGB=true
   )
  ]
  rendererClasses [&NiagaraRibbonRendererProperties]
  renderers [{rendererClass=&NiagaraRibbonRendererProperties}]
 }
 ToonTrail {
  emitterSpawnScript [&FN_RibbonIDOffsetUtility()]
  emitterUpdateScript [
   &EmitterState()
   &FN_SpawnAnimTrail()
  ]
  particleSpawnScript [
   &FN_InitializeAnimTrailBoneIndecies(First Bone or Socket Index=1, Second Bone or Socket Index=2)
   &FN_InitializeAnimTrail(Color=<rgba 0.4219 0.4219 0.4219 0.95>, Lifetime=0.16)
  ]
  particleUpdateScript [
   &SolveForcesAndVelocity(Clamp Velocity=false, Limit Acceleration=false)
   &FN_UpdateAnimTrail() {module=FN_UpdateAnimTrail001}
   &ScaleColor(
    Scale Alpha=&Multiply_Float(
     A=&FloatFromCurve(FloatCurve=DataInterface() {propertyValues {Curve {keys [<at 0.1 1 RCIM_Cubic>, {arriveTangent=-2.5031, tangentMode=RCTM_User, time=1, value=0}]}}})
     B=$System.AlphaFade
    )
    Scale RGB=&MakeVectorFromLinearColorRGB(Color=&DaySequenceTimeOfDayColor(Linear Color:LinearColor=<rgba 1 1 1 1>, Time Of Day Influence:NiagaraFloat=1))
    ScaleA=true
    ScaleRGB=true
   )
  ]
  rendererClasses [&NiagaraRibbonRendererProperties]
  renderers [{rendererClass=&NiagaraRibbonRendererProperties}]
 }
}
summary {
 systemName=NS_Pickaxe_ChiliAbide_Trail
 emitters [
  {emitterName=Soft_Trail, rendererClasses [&NiagaraRibbonRendererProperties]}
  {emitterName=ToonTrail, rendererClasses [&NiagaraRibbonRendererProperties]}
 ]
 userVariables [
  User.bIsFrontEnd {type=NiagaraBool}
  User.bIsFrontEndPreview {type=NiagaraBool}
  User.RibbonIDOffset {type=NiagaraInt32}
  User.SwingSectionIndex {type=NiagaraInt32}
 ]
}
